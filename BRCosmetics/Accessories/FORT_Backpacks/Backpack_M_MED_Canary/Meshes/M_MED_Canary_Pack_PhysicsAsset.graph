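# Physics asset for the Canary backpack cosmetic, providing capsule bodies and angular limits for its dangling straps and attachments.
# A two-capsule pack body anchors dyn_tomato, straps 02–05 and a three-link little-strap chain; dyn_tomato carries dyn_bag, dyn_slap and dyn_strap_01.
# Most large straps swing mainly on one axis, like hinges.
# dyn_strap_04 also links to dyn_strap_01, closing a loop through dyn_tomato and the pack.
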PhysicsAsset /BRCosmetics/Accessories/FORT_Backpacks/Backpack_M_MED_Canary/Meshes/M_MED_Canary_Pack_PhysicsAsset
bodies {
 dyn_bag [{shapeType=Capsule, length=1, radius=1, shapeName=dyn_bag_capsule}]
 dyn_little_strap_00 [{shapeType=Capsule, length=0, radius=0.3, shapeName=dyn_little_strap_00_capsule, center=<1 0 0>}]
 dyn_little_strap_01 [{shapeType=Capsule, length=0, radius=0.3, shapeName=dyn_little_strap_01_capsule, center=<1 0 0>}]
 dyn_little_strap_02 [{shapeType=Capsule, length=0, radius=0.3, shapeName=dyn_little_strap_02_capsule, center=<1 0 0>}]
 dyn_slap [{shapeType=Capsule, length=1, radius=1, shapeName=dyn_slap_capsule}]
 dyn_strap_01 [{shapeType=Capsule, length=8.5, radius=1, shapeName=dyn_strap_01_capsule, center=<6.5728 0 -0.7102>, rotation=<rot 80 -0.0006 -0.0001>}]
 dyn_strap_02 [{
   shapeType=Capsule
   length=11.1482
   radius=1.5772
   shapeName=dyn_strap_02_capsule
   center=<7.533 -0.4613 1.4195>
   rotation=<rot 0.0006 87.6597 -77.2686>
  }]
 dyn_strap_03 [{
   shapeType=Capsule
   length=2
   radius=2.299
   shapeName=dyn_strap_03_capsule
   center=<0.223 0.0764 -0.3941>
   rotation=<rot -0.5896 89.9538 -93.3433>
  }]
 dyn_strap_04 [{shapeType=Capsule, length=0, radius=2.25, shapeName=dyn_strap_04_capsule, rotation=<rot -45 0 0>}]
 dyn_strap_05 [{shapeType=Capsule, length=0, radius=2.25, shapeName=dyn_strap_05_capsule, rotation=<rot 90 0 0>}]
 dyn_tomato [{shapeType=Capsule, length=0.25, radius=4.25, shapeName=dyn_tomato_capsule, center=<4.9358 0 -0.9582>}]
 pack [
  {shapeType=Capsule, length=12.5, radius=7.75, shapeName=pack_capsule, center=<2.9556 -5.9805 -9.5176>, rotation=<rot -60 0.0006 -0.0007>}
  {shapeType=Capsule, length=15, radius=15, shapeName=pack_capsule, center=<0 13.7995 -19.8518>}
 ]
}
constraints [
 {bone1Name=dyn_tomato, bone2Name=pack, swing1LimitDegrees=5, swing2LimitDegrees=5, twistLimitDegrees=1}
 {bone1Name=dyn_strap_01, bone2Name=dyn_tomato, swing1LimitDegrees=1, swing2LimitDegrees=70, twistLimitDegrees=1}
 {bone1Name=dyn_strap_02, bone2Name=pack, swing1LimitDegrees=1, swing2LimitDegrees=45, twistLimitDegrees=1}
 {bone1Name=dyn_strap_03, bone2Name=pack, swing1LimitDegrees=10, swing2LimitDegrees=45, twistLimitDegrees=1}
 {bone1Name=dyn_little_strap_00, bone2Name=pack, swing1LimitDegrees=10, swing2LimitDegrees=10, twistLimitDegrees=10}
 {bone1Name=dyn_little_strap_01, bone2Name=dyn_little_strap_00, swing1LimitDegrees=20, swing2LimitDegrees=20, twistLimitDegrees=5}
 {bone1Name=dyn_little_strap_02, bone2Name=dyn_little_strap_01, swing1LimitDegrees=20, swing2LimitDegrees=20, twistLimitDegrees=5}
 {bone1Name=dyn_strap_04, bone2Name=pack, swing1LimitDegrees=10, swing2LimitDegrees=80, twistLimitDegrees=1}
 {bone1Name=dyn_strap_05, bone2Name=pack, swing1LimitDegrees=0, swing2LimitDegrees=80, twistLimitDegrees=1}
 {bone1Name=dyn_strap_04, bone2Name=dyn_strap_01, swing1LimitDegrees=10, swing2LimitDegrees=80, twistLimitDegrees=0}
 {bone1Name=dyn_bag, bone2Name=dyn_tomato, swing1LimitDegrees=1, swing2LimitDegrees=2, twistLimitDegrees=1}
 {bone1Name=dyn_slap, bone2Name=dyn_tomato, swing1LimitDegrees=1, swing2LimitDegrees=30, twistLimitDegrees=1}
]
